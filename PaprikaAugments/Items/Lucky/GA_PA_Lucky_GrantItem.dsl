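; GA_PA_Lucky_GrantItem is a gameplay ability for the Lucky augment that can grant the player one random item.
; GrantRandomItem builds three item entries from configurable items, with the first entry's count read from a gold amount ScalableFloat, and picks one at random.
; The pick is broken into definition and count, then granted as a world item definition.
; Activation and end events are unimplemented stubs, and nothing in this graph calls GrantRandomItem.
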(event Ability|EventActivateAbility
  (|CallParentFunction) ; This node is disabled and will not be called.\nDrag off pins to build functionality.
)

(event Ability|EventOnEndAbility
  (params (bWasCancelled :type Boolean))
  (|CallParentFunction bWasCancelled) ; This node is disabled and will not be called.\nDrag off pins to build functionality.
)

(fn GrantRandomItem
  (bind (_itemdefinition _itemguid _count) (Class|FortItemFunctionLibrary|BreakFortItemEntry (Utilities|Array|RandomArrayItem (Utilities|Array|MakeArray (Loot|CreateItemEntry (Variables|Default|GetItem%201) (Math|Float|Truncate (ScalableFloat|GetValueatLevel (Variables|Default|Getgold%20amount)))) (Loot|CreateItemEntry (Variables|Default|GetItem%202)) (Loot|CreateItemEntry (Variables|Default|GetItem%203))))))
  (CallFunction|GrantItem (Utilities|Casting|CastToFortWorldItemDefinition _itemdefinition) _count))
